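# Sound cue for the Shockwave Grenade explosion, layering close explosion and layer sounds with a distant variant under separate close and medium attenuations.
# The first quality level randomizes three waves per layer, with pitch variation on close layers and volume variation on distant.
# Other levels share a lighter branch with optimized waves and fewer variations, where a nested quality switch further trims the distant layer.
# Concurrency caps it at two instances, stopping the oldest.
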SoundCue /FNGameplayCues/Sounds/Fort_GamePlay_Sounds/Explosives/ShockwaveGrenade/SC_Grenade_Shockwave_Explosion
root=QualityLevel_0
nodes [
 {
  id=Attenuation_0
  children [Modulator_13]
  properties {attenuationSettings=/Game/Sounds/Attenuation/Fort_Explosion_Medium_Athena_Grenade_Att}
 }
 {id=Attenuation_1, children [Mixer_1], properties {attenuationSettings=/Game/Sounds/Attenuation/Fort_Explosion_Close_Athena_Grenade_Att}}
 {id=Attenuation_8, children [Modulator_6], properties {attenuationSettings=&Fort_Explosion_Medium_Athena_Grenade_Att}}
 {id=Attenuation_9, children [Mixer_3], properties {attenuationSettings=&Fort_Explosion_Close_Athena_Grenade_Att}}
 {id=Mixer_1, children [Modulator_0,Modulator_1], properties {inputVolume [1.1,1]}}
 {id=Mixer_2, children [Attenuation_1, Attenuation_0], properties {inputVolume [1,1]}}
 {id=Mixer_3, children [Modulator_4,Modulator_5], properties {inputVolume [1,1]}}
 {id=Mixer_4, children [Attenuation_9, Attenuation_8], properties {inputVolume [1,1]}}
 {id=Modulator_0, children [Random_0], properties {pitchMax=1.1, pitchMin=0.9}}
 {id=Modulator_1, children [Random_2], properties {pitchMax=1.1, pitchMin=0.9}}
 {id=Modulator_13, children [Random_8], properties {volumeMax=1.1, volumeMin=1}}
 {id=Modulator_4, children [WavePlayer_4], properties {pitchMax=1.1, pitchMin=0.9}}
 {id=Modulator_5, children [Random_10], properties {pitchMax=1.1, pitchMin=0.9}}
 {id=Modulator_6, children [QualityLevel_1], properties {volumeMax=1.1, volumeMin=1}}
 {id=QualityLevel_0, children [Mixer_2,Mixer_4,Mixer_4,Mixer_4]}
 {id=QualityLevel_1, children [Random_11,Random_11,WavePlayer_3,WavePlayer_3]}
 {id=Random_0, children [WavePlayer_0,WavePlayer_1,WavePlayer_2], properties {weights [1,1,1]}}
 {id=Random_10, children [WavePlayer_5, WavePlayer_20], properties {weights [1,1]}}
 {id=Random_11, children [WavePlayer_22, WavePlayer_23], properties {weights [1,1]}}
 {id=Random_2, children [WavePlayer_11, WavePlayer_12, WavePlayer_13], properties {weights [1,1,1]}}
 {id=Random_8, children [WavePlayer_14, WavePlayer_15, WavePlayer_16], properties {weights [1,1,1]}}
 {id=WavePlayer_0, properties {soundWaveAssetPtr=/Game/Sounds/Fort_GamePlay_Sounds/Explosives/ShockwaveGrenade/Grenade_Shockwave_Explode01}}
 {id=WavePlayer_1, properties {soundWaveAssetPtr=/Game/Sounds/Fort_GamePlay_Sounds/Explosives/ShockwaveGrenade/Grenade_Shockwave_Explode02}}
 {
  id=WavePlayer_11
  properties {soundWaveAssetPtr=/Game/Sounds/Fort_GamePlay_Sounds/Explosives/ShockwaveGrenade/Grenade_Shockwave_ExplodeLayer01}
 }
 {
  id=WavePlayer_12
  properties {soundWaveAssetPtr=/Game/Sounds/Fort_GamePlay_Sounds/Explosives/ShockwaveGrenade/Grenade_Shockwave_ExplodeLayer02}
 }
 {
  id=WavePlayer_13
  properties {soundWaveAssetPtr=/Game/Sounds/Fort_GamePlay_Sounds/Explosives/ShockwaveGrenade/Grenade_Shockwave_ExplodeLayer03}
 }
 {
  id=WavePlayer_14
  properties {soundWaveAssetPtr=/Game/Sounds/Fort_GamePlay_Sounds/Explosives/ShockwaveGrenade/Grenade_Shockwave_Explode_Distant01}
 }
 {
  id=WavePlayer_15
  properties {soundWaveAssetPtr=/Game/Sounds/Fort_GamePlay_Sounds/Explosives/ShockwaveGrenade/Grenade_Shockwave_Explode_Distant02}
 }
 {
  id=WavePlayer_16
  properties {soundWaveAssetPtr=/Game/Sounds/Fort_GamePlay_Sounds/Explosives/ShockwaveGrenade/Grenade_Shockwave_Explode_Distant03}
 }
 {id=WavePlayer_2, properties {soundWaveAssetPtr=/Game/Sounds/Fort_GamePlay_Sounds/Explosives/ShockwaveGrenade/Grenade_Shockwave_Explode03}}
 {
  id=WavePlayer_20
  properties {soundWaveAssetPtr=&Grenade_Shockwave_ExplodeLayer02}
 }
 {
  id=WavePlayer_22
  properties {soundWaveAssetPtr=&Grenade_Shockwave_Explode_Distant01}
 }
 {
  id=WavePlayer_23
  properties {soundWaveAssetPtr=&Grenade_Shockwave_Explode_Distant02}
 }
 {
  id=WavePlayer_3
  properties {soundWaveAssetPtr=&/Grenade_Shockwave_Explode_Distant_opt}
 }
 {
  id=WavePlayer_4
  properties {soundWaveAssetPtr=&/Grenade_Shockwave_Explode01_opt}
 }
 {
  id=WavePlayer_5
  properties {soundWaveAssetPtr=&/Grenade_Shockwave_ExplodeLayer01_opt}
 }
]
properties {
 bOverrideConcurrency=true
 maxDistance=9625
 soundClassObject=&Fort_Gadgets_Items
 volumeMultiplier=1.1
 attenuationOverrides {distanceAlgorithm=NaturalSound, falloffDistance=3800, lPFRadiusMax=3000, lPFRadiusMin=2000, attenuationShapeExtents {x=2200}}
 concurrencyOverrides {maxCount=2, resolutionRule=StopOldest}
}
